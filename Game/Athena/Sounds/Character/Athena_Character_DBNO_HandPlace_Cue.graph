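# Sound cue for the hand-placement foley of a downed (DBNO) character in Athena, reusing walk-on-dirt footstep waves.
# A quality-level node feeds the first two levels a five-wave random pool and the last two a three-wave pool, through a default Modulator.
# It uses third-person DBNO shuffle attenuation and the enemy foley sound class.
# Its concurrency override allows one instance per owner, stopping the oldest.
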SoundCue /Game/Athena/Sounds/Character/Athena_Character_DBNO_HandPlace_Cue
root=Modulator_3
nodes [
 {id=Modulator_3, children [QualityLevel_0]}
 {id=QualityLevel_0, children [Random_0,Random_0,Random_1,Random_1]}
 {id=Random_0, children [WavePlayer_0,WavePlayer_1,WavePlayer_2,WavePlayer_3,WavePlayer_4], properties {weights [1,1,1,1,1]}}
 {id=Random_1, children [WavePlayer_5,WavePlayer_6,WavePlayer_7], properties {bShouldExcludeFromBranchCulling=true, weights [1,1,1]}}
 {id=WavePlayer_0, properties {soundWaveAssetPtr=/Game/Sounds/Fort_Step_Sounds/Human/player_step_walk_dirt_01}}
 {id=WavePlayer_1, properties {soundWaveAssetPtr=/Game/Sounds/Fort_Step_Sounds/Human/player_step_walk_dirt_02}}
 {id=WavePlayer_2, properties {soundWaveAssetPtr=/Game/Sounds/Fort_Step_Sounds/Human/player_step_walk_dirt_03}}
 {id=WavePlayer_3, properties {soundWaveAssetPtr=/Game/Sounds/Fort_Step_Sounds/Human/player_step_walk_dirt_04}}
 {id=WavePlayer_4, properties {soundWaveAssetPtr=/Game/Sounds/Fort_Step_Sounds/Human/player_step_walk_dirt_05}}
 {id=WavePlayer_5, properties {soundWaveAssetPtr=&player_step_walk_dirt_01}}
 {id=WavePlayer_6, properties {soundWaveAssetPtr=&player_step_walk_dirt_02}}
 {id=WavePlayer_7, properties {soundWaveAssetPtr=&player_step_walk_dirt_03}}
]
properties {
 attenuationSettings=/Game/Sounds/Attenuation/Fort_Player_DBNO_Shuffle_3P_Athena_Att
 bOverrideConcurrency=true
 maxDistance=1500
 soundClassObject=&Fort_Player_Foley_3P_Enemy_Athena
 volumeMultiplier=0.8
 concurrencyOverrides {bLimitToOwner=true, maxCount=1, resolutionRule=StopOldest}
}
