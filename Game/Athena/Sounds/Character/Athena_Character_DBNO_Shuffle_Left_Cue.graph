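# Sound cue for the left-side shuffle foley of a downed (DBNO) Athena character.
# A Random node picks one of three equally weighted player run foley waves and feeds it through a Modulator.
# It plays quietly under the third-person enemy foley sound class with DBNO shuffle attenuation.
# A concurrency override caps simultaneous plays per owner, stopping the oldest.
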SoundCue /Game/Athena/Sounds/Character/Athena_Character_DBNO_Shuffle_Left_Cue
root=Modulator_3
nodes [
 {id=Modulator_3, children [Random_2]}
 {id=Random_2, children [WavePlayer_20, WavePlayer_19, WavePlayer_18], properties {bShouldExcludeFromBranchCulling=true, weights [1,1,1]}}
 {id=WavePlayer_18, properties {soundWaveAssetPtr=/Game/Sounds/Fort_Foley_Sounds/Moves/player_run_foley_07}}
 {id=WavePlayer_19, properties {soundWaveAssetPtr=/Game/Sounds/Fort_Foley_Sounds/Moves/player_run_foley_03}}
 {id=WavePlayer_20, properties {soundWaveAssetPtr=/Game/Sounds/Fort_Foley_Sounds/Moves/player_run_foley_02}}
]
properties {
 attenuationSettings=/Game/Sounds/Attenuation/Fort_Player_DBNO_Shuffle_3P_Athena_Att
 bOverrideConcurrency=true
 maxDistance=1500
 soundClassObject=&Fort_Player_Foley_3P_Enemy_Athena
 volumeMultiplier=0.2
 concurrencyOverrides {bLimitToOwner=true, maxCount=2, resolutionRule=StopOldest}
}
