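# Sound cue for the Figment variant of the last-player-killed sound, playing the Victory Royale music stinger.
# A single wave player feeds a Mixer, whose second input is empty.
# It uses the Victory Royale UI sound class with boosted volume, high priority and virtualization disabled.
# It references the LastPlayerKilled concurrency set and also carries single-instance, prevent-new override settings.
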SoundCue /Game/Athena/Sounds/Character/LastPlayerKilled_Figment_Cue
root=Mixer_3
nodes [
 {id=Mixer_3, children [WavePlayer_5,null], properties {inputVolume [1,0.75]}}
 {id=WavePlayer_5, properties {soundWaveAssetPtr=/Game/Sounds/Fort_Music/Stingers/Victory_Royale_01}}
]
properties {
 maxDistance=2097152
 priority=0.9
 soundClassObject=/Game/Sounds/FortSoundClassesAndModes/Fort_UI_VictoryRoyale_PSM
 virtualizationMode=Disabled
 volumeMultiplier=3
 concurrencyOverrides {maxCount=1, resolutionRule=PreventNew}
 concurrencySet [&/LastPlayerKilled_Conc]
}
