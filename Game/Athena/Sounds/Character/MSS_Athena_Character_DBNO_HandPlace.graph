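# One-shot mono MetaSound for third-person enemy DBNO crawl hand-place foley, limited to one instance per owner.
# On play it picks a random hand-place wave, and a DBNOSprintCheck compare's true branch triggers it plus a quieter clothing-run sweetener.
# DBNOSprintPitch raises hand-place pitch when sprinting, with random pitch and gain variation, and accumulated triggers signal completion.
# The Mobile page uses fewer waves, dropping the compare, sweetener and randomized pitch and gain.
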MetaSoundSource /Game/Athena/Sounds/Character/MSS_Athena_Character_DBNO_HandPlace
inputs [
 DBNOSprintCheck {type=Bool}
 DBNOSprintPitch {type=Float}
 In Array {type=WaveAsset:Array}
 In Array 1 {type=WaveAsset:Array}
 SprintSweetener {type=WaveAsset:Array}
 UE.Source.OnPlay {type=Trigger}
]
interfaces [UE.OutputFormat.Mono:1.0, UE.Source.OneShot:1.0, UE.Source:1.1]
outputs [
 "UE.OutputFormat.Mono.Audio:0" {type=Audio}
 UE.Source.OneShot.OnFinished {type=Trigger}
]
pages [
 Default {
  is_build_graph=true
  input_defaults {
   DBNOSprintCheck=null
   DBNOSprintPitch=5
   UE.Source.OnPlay=null
   In Array [
    /Game/Athena/Sounds/Character/DBNO/Crawling/DBNO_Crawl_HandPlace_05
    /Game/Athena/Sounds/Character/DBNO/Crawling/DBNO_Crawl_HandPlace_04
    /Game/Athena/Sounds/Character/DBNO/Crawling/DBNO_Crawl_HandPlace_03
    /Game/Athena/Sounds/Character/DBNO/Crawling/DBNO_Crawl_HandPlace_02
    /Game/Athena/Sounds/Character/DBNO/Crawling/DBNO_Crawl_HandPlace_01
   ]
   In Array 1 [/Game/Sounds/Fort_Test_Sounds/Tones/1000hz_Saw_Tone_Short]
   SprintSweetener [
    /Game/Sounds/Fort_Player/Clothing/Generic/Waves/Foley_Clothing_Generic_Run_08
    /Game/Sounds/Fort_Player/Clothing/Generic/Waves/Foley_Clothing_Generic_Run_07
    /Game/Sounds/Fort_Player/Clothing/Generic/Waves/Foley_Clothing_Generic_Run_06
    /Game/Sounds/Fort_Player/Clothing/Generic/Waves/Foley_Clothing_Generic_Run_05
    /Game/Sounds/Fort_Player/Clothing/Generic/Waves/Foley_Clothing_Generic_Run_04
    /Game/Sounds/Fort_Player/Clothing/Generic/Waves/Foley_Clothing_Generic_Run_03
    /Game/Sounds/Fort_Player/Clothing/Generic/Waves/Foley_Clothing_Generic_Run_02
    /Game/Sounds/Fort_Player/Clothing/Generic/Waves/Foley_Clothing_Generic_Run_01
   ]
  }
  nodes [
   n1 = InitVariable.Trigger()
   ArrayRandomGet_0 = Array.Random Get:WaveAsset:Array(In Array=^In Array, Next=^UE.Source.OnPlay)
   n2 = TriggerCompare.Clamp:Bool(A=^DBNOSprintCheck, Compare=@ArrayRandomGet_0.On Next) {also {defaults {B="<MetaSoundBuilderResult.SUCCEEDED: 0>"}}}
   ArrayRandomGet_1 = Array.Random Get:WaveAsset:Array(
    In Array=^SprintSweetener
    Next=VariableAccessor.Trigger(Variable=VariableMutator.Trigger(Variable=@n1.Variable, Value=@n2.True).Variable).Value
   ) {also {defaults {Weights []}}}
   WavePlayer_0 = UE.Wave Player:Mono(Wave Asset=@ArrayRandomGet_1.Value, Play=UE.RandomFloat(Next=@ArrayRandomGet_1.On Next) {also {defaults {Max=0.05, Min=-0.05}}}.On Next) {also {version [1,1]}}
   n3 = Convert.Float:Bool(Bool=^DBNOSprintCheck)
   n4 = UE.RandomFloat(Next=@n2.True) {also {defaults {Max=0.05, Min=-0.05}}}
   WavePlayer_1 = UE.Wave Player:Mono(
    Pitch Shift=UE.Add:Float(AdditionalOperands=UE.Multiply:Float(AdditionalOperands=^DBNOSprintPitch, PrimaryOperand=@n3.Float).Out, PrimaryOperand=@n4.Value).Out
    Wave Asset=@ArrayRandomGet_0.Value
    Play=@n4.On Next
   ) {also {version [1,1]}}
   AudioMixerAudioMixerMono2_0 = "AudioMixer.Audio Mixer (Mono, 2)"(
    Gain 1=UE.RandomFloat(Next=@WavePlayer_0.On Play) {also {defaults {Max=0.3, Min=0.25}}}.Value
    Gain 0=UE.RandomFloat(Next=@WavePlayer_1.On Play) {also {defaults {Max=1.05, Min=0.95}}}.Value
    In 1=@WavePlayer_0.Out Mono
    In 0=@WavePlayer_1.Out Mono
   )
   n5 = InitVariable.Trigger()
   TriggerAnyTriggerAny2_0 = "TriggerAny.Trigger Any (2)"(
    In 0=UE.Reroute(Value=@WavePlayer_0.On Finished).Value
    In 1=VariableAccessor.Trigger(Variable=VariableMutator.Trigger(Variable=@n5.Variable, Value=@n2.False).Variable).Value
   )
   VariableAccessorTrigger_0 = VariableAccessor.Trigger(Variable=VariableMutator.Trigger(Variable=InitVariable.Trigger().Variable, Value=@TriggerAnyTriggerAny2_0.Out).Variable)
   n6 = InitVariable.Trigger()
   TriggerAccumulatorTriggerAccumulate2_0 = "TriggerAccumulator.Trigger Accumulate (2)"(
    In 1=@VariableAccessorTrigger_0.Value
    In 0=VariableAccessor.Trigger(Variable=VariableMutator.Trigger(Variable=@n6.Variable, Value=@WavePlayer_1.On Finished).Variable).Value
   )
  ]
  links [
   {dst="@UE.OutputFormat.Mono.Audio:0", src=AudioMixerAudioMixerMono2_0.Out}
   {dst="@UE.Source.OneShot.OnFinished", src=TriggerAccumulatorTriggerAccumulate2_0.Out}
  ]
 }
 Mobile {
  input_defaults {
   In Array [
    &DBNO_Crawl_HandPlace_01
    &DBNO_Crawl_HandPlace_02
    &DBNO_Crawl_HandPlace_03
   ]
   SprintSweetener [
    &Foley_Clothing_Generic_Run_01
    &Foley_Clothing_Generic_Run_02
    &Foley_Clothing_Generic_Run_03
    &Foley_Clothing_Generic_Run_04
   ]
  }
  nodes [
   n1 = Convert.Float:Bool(Bool=^DBNOSprintCheck)
   ArrayRandomGet_0 = Array.Random Get:WaveAsset:Array(In Array=^In Array, Next=^UE.Source.OnPlay)
   WavePlayer_0 = UE.Wave Player:Mono(
    Pitch Shift=UE.Add:Float(AdditionalOperands=UE.Multiply:Float(AdditionalOperands=^DBNOSprintPitch, PrimaryOperand=@n1.Float).Out).Out
    Play=@ArrayRandomGet_0.On Next
    Wave Asset=@ArrayRandomGet_0.Value
   ) {also {version [1,1]}}
   AudioMixerAudioMixerMono2_0 = "AudioMixer.Audio Mixer (Mono, 2)"(In 0=@WavePlayer_0.Out Mono)
   InitVariable.Trigger()
   InitVariable.Trigger()
   InitVariable.Trigger()
   VariableAccessorTrigger_0 = VariableAccessor.Trigger(Variable=VariableMutator.Trigger(Variable=InitVariable.Trigger().Variable, Value=@WavePlayer_0.On Finished).Variable)
  ]
  links [
   {dst="@UE.OutputFormat.Mono.Audio:0", src=AudioMixerAudioMixerMono2_0.Out}
   {dst="@UE.Source.OneShot.OnFinished", src=VariableAccessorTrigger_0.Value}
  ]
 }
]
properties {
 assetImportData=&self:AssetImportData
 attenuationSettings=/Game/Sounds/Attenuation/Fort_Player_DBNO_Shuffle_3P_Athena_Att
 bOverrideConcurrency=true
 maxDistance=1500
 sampleRate=48000
 soundClassObject=&Fort_Player_Foley_3P_Enemy_Athena
 volume=0.75
 concurrencyOverrides {bLimitToOwner=true, maxCount=1, resolutionRule=StopOldest}
}
