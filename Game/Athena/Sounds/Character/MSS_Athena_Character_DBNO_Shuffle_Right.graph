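# Mono one-shot MetaSound source for the right-side DBNO crawl shuffle, using third-person enemy foley class and DBNO attenuation.
# On play, a random wave from the array feeds a Wave Player, with DBNOSprintCheck converted to float and multiplied by DBNOSprintPitch to shift pitch only when sprinting.
# A per-play random multiplier slightly varies gain.
# The Mobile page uses fewer samples and drops that gain randomization; concurrency is limited per owner, stopping oldest.
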MetaSoundSource /Game/Athena/Sounds/Character/MSS_Athena_Character_DBNO_Shuffle_Right
inputs [
 DBNOSprintCheck {type=Bool}
 DBNOSprintPitch {type=Float}
 In Array {type=WaveAsset:Array}
 UE.Source.OnPlay {type=Trigger}
]
interfaces [UE.OutputFormat.Mono:1.0, UE.Source.OneShot:1.0, UE.Source:1.1]
outputs [
 "UE.OutputFormat.Mono.Audio:0" {type=Audio}
 UE.Source.OneShot.OnFinished {type=Trigger}
]
pages [
 Default {
  is_build_graph=true
  input_defaults {
   DBNOSprintCheck=null
   DBNOSprintPitch=5
   UE.Source.OnPlay=null
   In Array [
    /Game/Athena/Sounds/Character/DBNO/Crawling/DBNO_Crawl_Shuffle_Right_01
    /Game/Athena/Sounds/Character/DBNO/Crawling/DBNO_Crawl_Shuffle_Right_02
    /Game/Athena/Sounds/Character/DBNO/Crawling/DBNO_Crawl_Shuffle_Right_03
    /Game/Athena/Sounds/Character/DBNO/Crawling/DBNO_Crawl_Shuffle_Right_04
    /Game/Athena/Sounds/Character/DBNO/Crawling/DBNO_Crawl_Shuffle_Right_05
   ]
  }
  nodes [
   InitVariable.Trigger()
   ArrayRandomGet_0 = Array.Random Get:WaveAsset:Array(In Array=^In Array, Next=^UE.Source.OnPlay)
   WavePlayer_0 = UE.Wave Player:Mono(
    Play=@ArrayRandomGet_0.On Next
    Wave Asset=@ArrayRandomGet_0.Value
    Pitch Shift=UE.Multiply:Float(AdditionalOperands=^DBNOSprintPitch, PrimaryOperand=Convert.Float:Bool(Bool=^DBNOSprintCheck).Float).Out
   ) {also {version [1,1]}}
   Multiply_0 = UE.Multiply:Audio by Float(
    AdditionalOperands=UE.RandomFloat(Next=@WavePlayer_0.On Play) {also {defaults {Max=1.05, Min=0.95}}}.Value
    PrimaryOperand=@WavePlayer_0.Out Mono
   )
  ]
  links [{dst="@UE.OutputFormat.Mono.Audio:0", src=Multiply_0.Out}, {dst="@UE.Source.OneShot.OnFinished", src=WavePlayer_0.On Finished}]
 }
 Mobile {
  input_defaults {In Array [
    &DBNO_Crawl_Shuffle_Right_01
    &DBNO_Crawl_Shuffle_Right_02
    &DBNO_Crawl_Shuffle_Right_03
   ]}
  nodes [
   InitVariable.Trigger()
   n1 = Convert.Float:Bool(Bool=^DBNOSprintCheck)
   ArrayRandomGet_0 = Array.Random Get:WaveAsset:Array(In Array=^In Array, Next=^UE.Source.OnPlay)
   WavePlayer_0 = UE.Wave Player:Mono(
    Pitch Shift=UE.Add:Float(AdditionalOperands=UE.Multiply:Float(AdditionalOperands=^DBNOSprintPitch, PrimaryOperand=@n1.Float).Out).Out
    Play=@ArrayRandomGet_0.On Next
    Wave Asset=@ArrayRandomGet_0.Value
   ) {also {version [1,1]}}
   Multiply_0 = UE.Multiply:Audio by Float(PrimaryOperand=@WavePlayer_0.Out Mono)
  ]
  links [{dst="@UE.OutputFormat.Mono.Audio:0", src=Multiply_0.Out}, {dst="@UE.Source.OneShot.OnFinished", src=WavePlayer_0.On Finished}]
 }
]
properties {
 assetImportData=&self:AssetImportData
 attenuationSettings=/Game/Sounds/Attenuation/Fort_Player_DBNO_Shuffle_3P_Athena_Att
 bOverrideConcurrency=true
 maxDistance=1500
 sampleRate=48000
 soundClassObject=&Fort_Player_Foley_3P_Enemy_Athena
 volume=0.15
 concurrencyOverrides {bLimitToOwner=true, maxCount=2, resolutionRule=StopOldest}
}
